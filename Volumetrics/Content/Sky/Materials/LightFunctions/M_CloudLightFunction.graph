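# Light function material that masks a light with a cloud shadow pattern.
# World XY position, divided by a scale and shifted by an offset, samples the red channel of an engine example noise texture at a parameter-set mip level.
# Bias subtraction clamped at zero, times Density, feeds a custom exp(-x) node for Beer-Lambert style attenuation.
# Contrast blends from unshadowed white to that attenuation.
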Material /Volumetrics/Content/Sky/Materials/LightFunctions/M_CloudLightFunction
graph {
 properties {materialDomain=MD_LightFunction, pixelDepthOffsetMode=PDOM_Legacy, usedShadingModels=Default Lit}
 nodes [
  n1 = Divide(A=WorldPosition().XYZ, B=Multiply(A=ScalarParameter() {defaultValue=100, parameterName=Mask Scale}) {constB=100000})
  n2 = VectorParameter() {parameterName=Offset, defaultValue=<rgba 0.5 0.5 0 0>}
  n3 = TextureSample!(
   UVs=ComponentMask(None=Subtract(A=@n1, B=@n2.RGB)) {g=true, r=true}
   Tex
   Level=ScalarParameter() {parameterName=2D Mask Mip}
   Apply View MipBias
  ) {
   mipValueMode=TMVM_MipLevel
   samplerType=SAMPLERTYPE_LinearColor
   texture=/Engine/Functions/Engine_MaterialFunctions02/ExampleContent/Textures/OffsetNoiseDistanceFields_ForNormals
  }
  n4 = ScalarParameter() {defaultValue=0.45, parameterName=Bias}
  MaterialExpressionMultiply_0 = Multiply(A=Max(A=Subtract(A=@n3.R, B=@n4)) {constB=0}, B=ScalarParameter() {defaultValue=16, parameterName=Density})
  n5 = Custom!(x=@MaterialExpressionMultiply_0) {
   code="return exp(-x);"
   description=Exp
   outputType=CMOT_Float1
   inputs [{inputName=x, input {expression=&self:MaterialExpressionMultiply_0}}]
  }
  LinearInterpolate(B=@n5, Alpha=ScalarParameter() {defaultValue=1, parameterName=Contrast}) {constA=1}
 ]
}
